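# Animation Blueprint for the Avocado Eater Nug creature, layering its own animation from alien_root onto a pose copied from the attached parent mesh.
# A conduit-hub state machine picks Intro, ParasiteIntro, LayEgg, Alert, terminal Death or fallback idle/run Locomotion.
# Named bool anim input events set the state flags; Speed sets idle.
# Chained AnimDynamics on four two-bone eye stalks and a RigidBody node add secondary motion before inertialization.
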AnimBlueprint /MotherGameplay/Items/AvacadoEaterBird/Animation/AvocadoEater_Nug/AvocadoEater_Nug_AnimBP
target_skeleton=/MotherGameplay/Items/AvacadoEaterBird/Meshes/Hitchhiker_Nug/Hitchhiker_Nug_Mammal_Skeleton
graphs [
 AnimGraph {nodes [
   n1 = Copy Pose From Mesh() {bCopyCurves=true, bUseAttachedParent=true}
   AnimGraphNode_StateMachine_0 = AvocadoEater_Nug_StateMachine() {is=title AnimGraphNode_StateMachine}
   n2 = Anim Dynamics - Bone: alien_eye_base_fl(
    ComponentPose=Local To Component(LocalPose=Layered blend per bone(BasePose=@n1, BlendPoses_0=@AnimGraphNode_StateMachine_0) {layerSetup [{branchFilters [{boneName=alien_root}]}]})
    Alpha=1
    ExternalForce="0,0,0"
   ) {
    bGravityOverrideInSimSpace=true
    bUseGravityOverride=true
    alphaBoolBlend {blendOption=Linear}
    boundBone {boneName=alien_eye_base_fl}
    componentAppliedLinearAccClamp=<2000 2000 2000>
    componentLinearAccScale=<0.3 0.3 0.35>
    componentLinearVelScale=<0.3 0.3 0.35>
    gravityOverride {z=-980}
    physicsBodyDefinitions [{
      sphereCollisionRadius=0
      boundBone {boneName=alien_eye_base_fl}
      boxExtents=<5 5 5>
      constraintSetup {angularLimitsMax=<10 10 10>, angularLimitsMin=<0 0 0>, angularTarget=<0 0 0>}
      localJointOffset=<0.5 0 2.5>
     }]
    relativeSpaceBone {boneName=root}
   }
   n3 = Anim Dynamics - Bone: alien_eye01_fl(ComponentPose=@n2, Alpha=1, ExternalForce="0,0,0") {
    bGravityOverrideInSimSpace=true
    bUseGravityOverride=true
    alphaBoolBlend {blendOption=Linear}
    boundBone {boneName=alien_eye01_fl}
    componentAppliedLinearAccClamp=<2000 2000 2000>
    componentLinearAccScale=<0.3 0.3 0.35>
    componentLinearVelScale=<0.3 0.3 0.35>
    gravityOverride {z=-980}
    physicsBodyDefinitions [{
      sphereCollisionRadius=0
      boundBone {boneName=alien_eye01_fl}
      boxExtents=<5 5 5>
      constraintSetup {angularLimitsMax=<10 10 10>, angularLimitsMin=<0 0 0>, angularTarget=<0 0 0>}
      localJointOffset=<0.5 0 2.5>
     }]
    relativeSpaceBone {boneName=root}
   }
   n4 = Anim Dynamics - Bone: alien_eye_base_fr(ComponentPose=@n3, Alpha=1, ExternalForce="0,0,0") {
    bGravityOverrideInSimSpace=true
    bUseGravityOverride=true
    alphaBoolBlend {blendOption=Linear}
    boundBone {boneName=alien_eye_base_fr}
    componentAppliedLinearAccClamp=<2000 2000 2000>
    componentLinearAccScale=<0.3 0.3 0.35>
    componentLinearVelScale=<0.3 0.3 0.35>
    gravityOverride {z=-980}
    physicsBodyDefinitions [{
      sphereCollisionRadius=0
      boundBone {boneName=alien_eye_base_fr}
      boxExtents=<5 5 5>
      constraintSetup {angularLimitsMax=<25 25 25>, angularLimitsMin=<0 0 0>, angularTarget=<0 0 0>}
      localJointOffset=<0 0 2.5>
     }]
    relativeSpaceBone {boneName=root}
   }
   n5 = Anim Dynamics - Bone: alien_eye01_fr(ComponentPose=@n4, Alpha=1, ExternalForce="0,0,0") {
    bGravityOverrideInSimSpace=true
    bUseGravityOverride=true
    alphaBoolBlend {blendOption=Linear}
    boundBone {boneName=alien_eye01_fr}
    componentAppliedLinearAccClamp=<2000 2000 2000>
    componentLinearAccScale=<0.3 0.3 0.35>
    componentLinearVelScale=<0.3 0.3 0.35>
    gravityOverride {z=-980}
    physicsBodyDefinitions [{
      sphereCollisionRadius=0
      boundBone {boneName=alien_eye01_fr}
      boxExtents=<5 5 5>
      constraintSetup {angularLimitsMax=<15 15 15>, angularLimitsMin=<0 0 0>, angularTarget=<0 0 0>}
      localJointOffset=<0 0 2.5>
     }]
    relativeSpaceBone {boneName=root}
   }
   n6 = Anim Dynamics - Bone: alien_eye_base_bl(ComponentPose=@n5, Alpha=1, ExternalForce="0,0,0") {
    bGravityOverrideInSimSpace=true
    bUseGravityOverride=true
    alphaBoolBlend {blendOption=Linear}
    boundBone {boneName=alien_eye_base_bl}
    componentAppliedLinearAccClamp=<2000 2000 2000>
    componentLinearAccScale=<0.3 0.3 0.35>
    componentLinearVelScale=<0.3 0.3 0.35>
    gravityOverride {z=-980}
    physicsBodyDefinitions [{
      sphereCollisionRadius=0
      boundBone {boneName=alien_eye_base_bl}
      boxExtents=<5 5 5>
      constraintSetup {angularLimitsMax=<25 25 25>, angularLimitsMin=<0 0 0>, angularTarget=<0 0 0>}
      localJointOffset=<0 0 2.5>
     }]
    relativeSpaceBone {boneName=root}
   }
   n7 = Anim Dynamics - Bone: alien_eye01_bl(ComponentPose=@n6, Alpha=1, ExternalForce="0,0,0") {
    bGravityOverrideInSimSpace=true
    bUseGravityOverride=true
    alphaBoolBlend {blendOption=Linear}
    boundBone {boneName=alien_eye01_bl}
    componentAppliedLinearAccClamp=<2000 2000 2000>
    componentLinearAccScale=<0.3 0.3 0.35>
    componentLinearVelScale=<0.3 0.3 0.35>
    gravityOverride {z=-980}
    physicsBodyDefinitions [{
      sphereCollisionRadius=0
      boundBone {boneName=alien_eye01_bl}
      boxExtents=<5 5 5>
      constraintSetup {angularLimitsMax=<15 15 15>, angularLimitsMin=<0 0 0>, angularTarget=<0 0 0>}
      localJointOffset=<0 0 2.5>
     }]
    relativeSpaceBone {boneName=root}
   }
   n8 = Anim Dynamics - Bone: alien_eye_base_br(ComponentPose=@n7, Alpha=1, ExternalForce="0,0,0") {
    bGravityOverrideInSimSpace=true
    bUseGravityOverride=true
    alphaBoolBlend {blendOption=Linear}
    boundBone {boneName=alien_eye_base_br}
    componentAppliedLinearAccClamp=<2000 2000 2000>
    componentLinearAccScale=<0.3 0.3 0.35>
    componentLinearVelScale=<0.3 0.3 0.35>
    gravityOverride {z=-980}
    physicsBodyDefinitions [{
      sphereCollisionRadius=0
      boundBone {boneName=alien_eye_base_br}
      boxExtents=<5 5 5>
      constraintSetup {angularLimitsMax=<25 25 25>, angularLimitsMin=<0 0 0>, angularTarget=<0 0 0>}
      localJointOffset=<0 0 2.5>
     }]
    relativeSpaceBone {boneName=root}
   }
   n9 = Anim Dynamics - Bone: alien_eye01_br(ComponentPose=@n8, Alpha=1, ExternalForce="0,0,0") {
    bGravityOverrideInSimSpace=true
    bUseGravityOverride=true
    alphaBoolBlend {blendOption=Linear}
    boundBone {boneName=alien_eye01_br}
    componentAppliedLinearAccClamp=<2000 2000 2000>
    componentLinearAccScale=<0.3 0.3 0.35>
    componentLinearVelScale=<0.3 0.3 0.35>
    gravityOverride {z=-980}
    physicsBodyDefinitions [{
      sphereCollisionRadius=0
      boundBone {boneName=alien_eye01_br}
      boxExtents=<5 5 5>
      constraintSetup {angularLimitsMax=<15 15 15>, angularLimitsMin=<0 0 0>, angularTarget=<0 0 0>}
      localJointOffset=<0 0 2.5>
     }]
    relativeSpaceBone {boneName=root}
   }
   n10 = RigidBody(ComponentPose=@n9) {
    bOverrideWorldGravity=true
    overridePhysicsAsset=/MotherGameplay/Items/AvacadoEaterBird/Meshes/Hitchhiker_Nug/Hitchhiker_Nug_Mammal_Physics
    simulationSpace=BaseBoneSpace
    alphaBoolBlend {blendOption=Linear}
    baseBoneRef {boneName=root}
    componentAppliedLinearAccClamp=<2000 2000 2000>
    componentLinearAccScale=<0.3 0.3 0.35>
    componentLinearVelScale=<0.3 0.3 0.35>
    overrideWorldGravity {z=-980}
   }
   Output Pose(Result=Inertialization!(Source:Pose Link Structure=Component To Local(ComponentPose=@n10)) -> (Pose:Pose Link Structure) {is=title AnimGraphNode_Inertialization}) {layerGroup=None}
  ]}
 AvocadoEater_Nug_StateMachine {
  entry=DirtyHackState
  path=AnimGraph.AnimGraphNode_StateMachine_0.AvocadoEater_Nug_StateMachine
  ids=id
  nodes [
   AnimStateConduitNode_0 = Conduit()
   AnimStateNode_0 = DirtyHackState() {is=title AnimStateNode}
   AnimStateNode_1 = Locomotion() {is=title AnimStateNode}
   AnimStateNode_2 = LayEgg() {is=title AnimStateNode}
   AnimStateNode_3 = Death() {is=title AnimStateNode}
   AnimStateNode_4 = ParasiteIntro() {is=title AnimStateNode}
   AnimStateNode_5 = Alert() {is=title AnimStateNode}
   AnimStateNode_6 = Intro() {is=title AnimStateNode}
   AnimStateTransitionNode_0 = AnimStateTransitionNode()
   AnimStateTransitionNode_1 = AnimStateTransitionNode() {crossfadeDuration=0.3, logicType=TLT_Inertialization, priorityOrder=100}
   AnimStateTransitionNode_10 = AnimStateTransitionNode() {crossfadeDuration=0.3, logicType=TLT_Inertialization}
   AnimStateTransitionNode_11 = AnimStateTransitionNode()
   AnimStateTransitionNode_2 = AnimStateTransitionNode() {crossfadeDuration=0.3, logicType=TLT_Inertialization, priorityOrder=2}
   AnimStateTransitionNode_3 = AnimStateTransitionNode()
   AnimStateTransitionNode_4 = AnimStateTransitionNode()
   AnimStateTransitionNode_5 = AnimStateTransitionNode() {crossfadeDuration=0.3, logicType=TLT_Inertialization, priorityOrder=5}
   AnimStateTransitionNode_6 = AnimStateTransitionNode() {crossfadeDuration=0.3, logicType=TLT_Inertialization, priorityOrder=3}
   AnimStateTransitionNode_7 = AnimStateTransitionNode()
   AnimStateTransitionNode_8 = AnimStateTransitionNode() {logicType=TLT_Inertialization, priorityOrder=4}
   AnimStateTransitionNode_9 = AnimStateTransitionNode()
  ]
  links [
   {
    dst=Conduit
    rule=AnimGraph.AnimGraphNode_StateMachine_0.AvocadoEater_Nug_StateMachine.AnimStateTransitionNode_9.Transition
    src=Alert
    via=AnimStateTransitionNode_9
   }
   {
    dst=Alert
    rule=AnimGraph.AnimGraphNode_StateMachine_0.AvocadoEater_Nug_StateMachine.AnimStateTransitionNode_8.Transition
    src=Conduit
    via=AnimStateTransitionNode_8
   }
   {
    dst=Death
    rule=AnimGraph.AnimGraphNode_StateMachine_0.AvocadoEater_Nug_StateMachine.AnimStateTransitionNode_5.Transition
    src=Conduit
    via=AnimStateTransitionNode_5
   }
   {
    dst=Intro
    rule=AnimGraph.AnimGraphNode_StateMachine_0.AvocadoEater_Nug_StateMachine.AnimStateTransitionNode_10.Transition
    src=Conduit
    via=AnimStateTransitionNode_10
   }
   {
    dst=LayEgg
    rule=AnimGraph.AnimGraphNode_StateMachine_0.AvocadoEater_Nug_StateMachine.AnimStateTransitionNode_2.Transition
    src=Conduit
    via=AnimStateTransitionNode_2
   }
   {
    dst=Locomotion
    rule=AnimGraph.AnimGraphNode_StateMachine_0.AvocadoEater_Nug_StateMachine.AnimStateTransitionNode_1.Transition
    src=Conduit
    via=AnimStateTransitionNode_1
   }
   {
    dst=ParasiteIntro
    rule=AnimGraph.AnimGraphNode_StateMachine_0.AvocadoEater_Nug_StateMachine.AnimStateTransitionNode_6.Transition
    src=Conduit
    via=AnimStateTransitionNode_6
   }
   {
    dst=Conduit
    rule=AnimGraph.AnimGraphNode_StateMachine_0.AvocadoEater_Nug_StateMachine.AnimStateTransitionNode_0.Transition
    src=DirtyHackState
    via=AnimStateTransitionNode_0
   }
   {
    dst=Conduit
    rule=AnimGraph.AnimGraphNode_StateMachine_0.AvocadoEater_Nug_StateMachine.AnimStateTransitionNode_11.Transition
    src=Intro
    via=AnimStateTransitionNode_11
   }
   {
    dst=Conduit
    rule=AnimGraph.AnimGraphNode_StateMachine_0.AvocadoEater_Nug_StateMachine.AnimStateTransitionNode_4.Transition
    src=LayEgg
    via=AnimStateTransitionNode_4
   }
   {
    dst=Conduit
    rule=AnimGraph.AnimGraphNode_StateMachine_0.AvocadoEater_Nug_StateMachine.AnimStateTransitionNode_3.Transition
    src=Locomotion
    via=AnimStateTransitionNode_3
   }
   {
    dst=Conduit
    rule=AnimGraph.AnimGraphNode_StateMachine_0.AvocadoEater_Nug_StateMachine.AnimStateTransitionNode_7.Transition
    src=ParasiteIntro
    via=AnimStateTransitionNode_7
   }
  ]
 }
 Conduit {
  path=AnimGraph.AnimGraphNode_StateMachine_0.AvocadoEater_Nug_StateMachine.AnimStateConduitNode_0.Conduit
  nodes [Result(bCanEnterTransition=true)]
 }
 DirtyHackState {
  path=AnimGraph.AnimGraphNode_StateMachine_0.AvocadoEater_Nug_StateMachine.AnimStateNode_0.DirtyHackState
  nodes [Output Animation Pose(Result=Local Space Ref Pose())]
 }
 Locomotion {
  path=AnimGraph.AnimGraphNode_StateMachine_0.AvocadoEater_Nug_StateMachine.AnimStateNode_1.Locomotion
  nodes [Output Animation Pose(Result=Blend Poses by bool(
     BlendPose_0=Sequence Player 'AvocadoEater_Nug_Idle'() {sequence=&/AvocadoEater_Nug_Idle}
     BlendPose_1=Sequence Player 'AvocadoEater_Nug_Run_N'() {sequence=&/AvocadoEater_Nug_Run_N}
     BlendTime_0=0.5
     BlendTime_1=0.5
     bActiveValue=Get IsIdle() -> (IsIdle:bool)
    ) {blendType=Linear, transitionType=Inertialization})]
 }
 LayEgg {
  path=AnimGraph.AnimGraphNode_StateMachine_0.AvocadoEater_Nug_StateMachine.AnimStateNode_2.LayEgg
  nodes [Output Animation Pose(Result=Sequence Player 'AvocadoEater_Nug_ParasiteLayEgg'() {sequence=&/AvocadoEater_Nug_ParasiteLayEgg})]
 }
 Death {
  path=AnimGraph.AnimGraphNode_StateMachine_0.AvocadoEater_Nug_StateMachine.AnimStateNode_3.Death
  nodes [Output Animation Pose(Result=Sequence Player 'AvocadoEater_Nug_Death'() {bLoopAnimation=false, sequence=&/AvocadoEater_Nug_Death})]
 }
 ParasiteIntro {
  path=AnimGraph.AnimGraphNode_StateMachine_0.AvocadoEater_Nug_StateMachine.AnimStateNode_4.ParasiteIntro
  nodes [Output Animation Pose(Result=Sequence Player 'AvocadoEater_Nug_ParasiteLayEgg'() {sequence=&AvocadoEater_Nug_ParasiteLayEgg})]
 }
 Alert {
  path=AnimGraph.AnimGraphNode_StateMachine_0.AvocadoEater_Nug_StateMachine.AnimStateNode_5.Alert
  nodes [Output Animation Pose(Result=Sequence Player 'AvocadoEater_Nug_Alert'() {sequence=&/AvocadoEater_Nug_Alert})]
 }
 Intro {
  path=AnimGraph.AnimGraphNode_StateMachine_0.AvocadoEater_Nug_StateMachine.AnimStateNode_6.Intro
  nodes [Output Animation Pose(Result=Sequence Player 'AvocadoEater_Nug_ParasiteInto'() {sequence=&/AvocadoEater_Nug_ParasiteInto})]
 }
 Transition {
  path=AnimGraph.AnimGraphNode_StateMachine_0.AvocadoEater_Nug_StateMachine.AnimStateTransitionNode_0.Transition
  nodes [Result(bCanEnterTransition=true)]
 }
 Transition {
  path=AnimGraph.AnimGraphNode_StateMachine_0.AvocadoEater_Nug_StateMachine.AnimStateTransitionNode_1.Transition
  nodes [Result(bCanEnterTransition=true)]
 }
 Transition {
  path=AnimGraph.AnimGraphNode_StateMachine_0.AvocadoEater_Nug_StateMachine.AnimStateTransitionNode_10.Transition
  nodes [Result(bCanEnterTransition=Get IsIntro() -> (IsIntro:bool))]
 }
 Transition {
  path=AnimGraph.AnimGraphNode_StateMachine_0.AvocadoEater_Nug_StateMachine.AnimStateTransitionNode_11.Transition
  nodes [Result(bCanEnterTransition=NOT Boolean(A=Get IsIntro()))]
 }
 Transition {
  path=AnimGraph.AnimGraphNode_StateMachine_0.AvocadoEater_Nug_StateMachine.AnimStateTransitionNode_2.Transition
  nodes [Result(bCanEnterTransition=Get IslayingEgg() -> (IslayingEgg:bool))]
 }
 Transition {
  path=AnimGraph.AnimGraphNode_StateMachine_0.AvocadoEater_Nug_StateMachine.AnimStateTransitionNode_3.Transition
  nodes [Result(bCanEnterTransition=true)]
 }
 Transition {
  path=AnimGraph.AnimGraphNode_StateMachine_0.AvocadoEater_Nug_StateMachine.AnimStateTransitionNode_4.Transition
  nodes [Result(bCanEnterTransition=NOT Boolean(A=Get IslayingEgg()))]
 }
 Transition {
  path=AnimGraph.AnimGraphNode_StateMachine_0.AvocadoEater_Nug_StateMachine.AnimStateTransitionNode_5.Transition
  nodes [Result(bCanEnterTransition=Get IsDeath() -> (IsDeath:bool))]
 }
 Transition {
  path=AnimGraph.AnimGraphNode_StateMachine_0.AvocadoEater_Nug_StateMachine.AnimStateTransitionNode_6.Transition
  nodes [Result(bCanEnterTransition=Get IsParasiteIntro() -> (IsParasiteIntro:bool))]
 }
 Transition {
  path=AnimGraph.AnimGraphNode_StateMachine_0.AvocadoEater_Nug_StateMachine.AnimStateTransitionNode_7.Transition
  nodes [Result(bCanEnterTransition=NOT Boolean(A=Get IsParasiteIntro()))]
 }
 Transition {
  path=AnimGraph.AnimGraphNode_StateMachine_0.AvocadoEater_Nug_StateMachine.AnimStateTransitionNode_8.Transition
  nodes [Result(bCanEnterTransition=Get IsAlert() -> (IsAlert:bool))]
 }
 Transition {
  path=AnimGraph.AnimGraphNode_StateMachine_0.AvocadoEater_Nug_StateMachine.AnimStateTransitionNode_9.Transition
  nodes [Result(bCanEnterTransition=NOT Boolean(A=Get IsAlert()))]
 }
 EventGraph {nodes [
   Event BlueprintUpdateAnimation()
   > n1 = Branch(Condition="float < float"!(A:double=Get Speed() -> (Speed:float), B:double=200) -> (ReturnValue:bool) {is=title K2Node_CallFunction})
   > Set IsIdle(IsIdle:bool=true) -> (Output_Get:bool)
   Set IsIdle(execute=@n1.else, IsIdle=false)
   n2 = Event BlueprintProcessAnimInputEvent()
   > n3 = Branch(Condition=IsNamed(self=@n2.AnimInputEvent, InEventName=aie.Nug.ParasiteIntro))
   > n4 = Cast To FortAnimInputEvent_Bool!(execute, Object=@n2.AnimInputEvent) -> (then, CastFailed, AsFort Anim Input Event Bool:Fort Anim Input Event Bool Object Reference) {is=title K2Node_DynamicCast}
   > Set IsParasiteIntro(IsParasiteIntro:bool=Get bValue!(self:Fort Anim Input Event Bool Object Reference=@n4) -> (bValue:bool)) -> (Output_Get:bool)
   n5 = Branch(execute=@n3.else, Condition=IsNamed(self=@n2.AnimInputEvent, InEventName=aie.Nug.Death))
   > n6 = Cast To FortAnimInputEvent_Bool(Object=@n2.AnimInputEvent)
   > Set IsDeath(IsDeath:bool=Get bValue(self=@n6)) -> (Output_Get:bool)
   n7 = Branch(execute=@n5.else, Condition=IsNamed(self=@n2.AnimInputEvent, InEventName=aie.Nug.LayEgg))
   > n8 = Cast To FortAnimInputEvent_Bool(Object=@n2.AnimInputEvent)
   > Set IslayingEgg(IslayingEgg:bool=Get bValue(self=@n8)) -> (Output_Get:bool)
   n9 = Branch(execute=@n7.else, Condition=IsNamed(self=@n2.AnimInputEvent, InEventName=aie.Nug.Alert))
   > n10 = Cast To FortAnimInputEvent_Bool(Object=@n2.AnimInputEvent)
   > Set IsAlert(IsAlert:bool=Get bValue(self=@n10)) -> (Output_Get:bool)
   Branch(execute=@n9.else, Condition=IsNamed(self=@n2.AnimInputEvent, InEventName=aie.Nug.Intro))
   > n11 = Cast To FortAnimInputEvent_Bool(Object=@n2.AnimInputEvent)
   > Set IsIntro(IsIntro:bool=Get bValue(self=@n11)) -> (Output_Get:bool)
   TryGetPawnOwner()
  ]}
]
variables [
 IsAlert {type=boolean}
 IsDeath {type=boolean}
 IsIdle {type=boolean}
 IsIntro {type=boolean}
 IslayingEgg {type=boolean}
 IsParasiteIntro {type=boolean}
]
